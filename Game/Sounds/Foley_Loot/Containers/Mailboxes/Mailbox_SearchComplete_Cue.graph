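# Sound cue for completing a mailbox search, routed to the Fort_Loot_Collect sound class with prop search attenuation.
# A random node picks one of three mailbox search waves.
# An enveloper fades the pick out in under a second, then a modulator applies slight pitch and volume variation.
# A concurrency override caps simultaneous instances, stopping the farthest before blocking new ones.
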SoundCue /Game/Sounds/Foley_Loot/Containers/Mailboxes/Mailbox_SearchComplete_Cue
root=Modulator_15
nodes [
 {
  id=Enveloper_0
  children [Random_1]
  properties {pitchCurve {editorCurveData {keys [{time=0}]}}, volumeCurve {editorCurveData {keys [<at 0 1>,<at 0.64 0>]}}}
 }
 {id=Modulator_15, children [Enveloper_0], properties {pitchMax=1.07, pitchMin=0.92, volumeMax=1.07, volumeMin=0.92}}
 {id=Random_1, children [WavePlayer_68, WavePlayer_70, WavePlayer_72], properties {weights [1,1,1]}}
 {id=WavePlayer_68, properties {soundWaveAssetPtr=&/Mailbox_Search01}}
 {id=WavePlayer_70, properties {soundWaveAssetPtr=&/Mailbox_Search03}}
 {id=WavePlayer_72, properties {soundWaveAssetPtr=&/Mailbox_Search05}}
]
properties {
 attenuationSettings=/Game/Sounds/Attenuation/Fort_Prop_Search_Att
 bOverrideConcurrency=true
 maxDistance=1300
 soundClassObject=&Fort_Loot_Collect
 volumeMultiplier=0.7
 concurrencyOverrides {maxCount=4, resolutionRule=StopFarthestThenPreventNew}
}
