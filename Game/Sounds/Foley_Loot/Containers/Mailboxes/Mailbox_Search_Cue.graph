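# Sound cue for searching a mailbox container, routed to the loot-collect sound class with prop-search attenuation.
# A mixer layers a metal mailbox stereo wave with two branches from one pitch- and volume-modulated random pick of three search waves.
# One branch starts immediately, while the other reuses the same modulator behind a delay of just over a second at reduced volume.
# A concurrency override prevents new instances once the limit is reached.
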SoundCue /Game/Sounds/Foley_Loot/Containers/Mailboxes/Mailbox_Search_Cue
root=Mixer_15
nodes [
 {id=Delay_32, children [Modulator_18], properties {delayMax=1.2, delayMin=1.1}}
 {id=Mixer_15, children [Modulator_18, Delay_32, WavePlayer_67], properties {inputVolume [1,0.7,1]}}
 {id=Modulator_18, children [Random_0], properties {pitchMax=1.07, pitchMin=0.92, volumeMax=1.07, volumeMin=0.92}}
 {id=Random_0, children [WavePlayer_59, WavePlayer_61, WavePlayer_63], properties {weights [1,1,1]}}
 {id=WavePlayer_59, properties {soundWaveAssetPtr=&/Mailbox_Search01}}
 {id=WavePlayer_61, properties {soundWaveAssetPtr=&/Mailbox_Search03}}
 {id=WavePlayer_63, properties {soundWaveAssetPtr=&/Mailbox_Search05}}
 {id=WavePlayer_67, properties {soundWaveAssetPtr=&/Metal_Mailbox_Stereo}}
]
properties {
 attenuationSettings=/Game/Sounds/Attenuation/Fort_Prop_Search_Att
 bOverrideConcurrency=true
 maxDistance=1300
 soundClassObject=&Fort_Loot_Collect
 volumeMultiplier=0.9
 concurrencyOverrides {resolutionRule=PreventNew}
}
