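# Sound cue for the end of a newspaper box loot search, layering the Fort_UI_NoMoreLoot sound with search foley.
# One of two NewspaperBox_Search waves is picked at random, faded out quickly by an Enveloper and varied slightly in pitch and volume by a Modulator.
# A Mixer combines this with the slightly lowered UI sound.
# It uses prop search attenuation, the Fort_Loot_Collect class and a concurrency override that stops the farthest instances.
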SoundCue /Game/Sounds/Foley_Loot/Containers/Mailboxes/NewspaperBox_SearchEnd_Cue
root=Mixer_12
nodes [
 {
  id=Enveloper_4
  children [Random_10]
  properties {pitchCurve {editorCurveData {keys [{time=0}]}}, volumeCurve {editorCurveData {keys [<at 0 1>,<at 0.64 0>]}}}
 }
 {id=Mixer_12, children [WavePlayer_43, Modulator_15], properties {inputVolume [0.8,1]}}
 {id=Modulator_15, children [Enveloper_4], properties {pitchMax=1.07, pitchMin=0.92, volumeMax=1.07, volumeMin=0.92}}
 {id=Random_10, children [WavePlayer_36, WavePlayer_37], properties {weights [1,1]}}
 {id=WavePlayer_36, properties {soundWaveAssetPtr=&/NewspaperBox_Search01}}
 {id=WavePlayer_37, properties {soundWaveAssetPtr=&/NewspaperBox_Search02}}
 {id=WavePlayer_43, properties {soundWaveAssetPtr=/Game/Sounds/Fort_UI_Sounds/Gameplay/Fort_UI_NoMoreLoot}}
]
properties {
 attenuationSettings=/Game/Sounds/Attenuation/Fort_Prop_Search_Att
 bOverrideConcurrency=true
 maxDistance=1300
 soundClassObject=&Fort_Loot_Collect
 volumeMultiplier=0.55
 concurrencyOverrides {maxCount=4, resolutionRule=StopFarthestThenPreventNew}
}
