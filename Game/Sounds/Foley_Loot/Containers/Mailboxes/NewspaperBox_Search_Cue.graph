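# Sound cue for searching a newspaper box loot container, played with prop search attenuation and a concurrency override that prevents new instances.
# A quality-level node selects between two layered mixes and two lighter branches that each play a single optimized wave.
# Both mixes layer a random newspaper box search, varied in pitch and volume, with gear foley.
# The first mix randomizes three gear clips, while the second uses one.
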SoundCue /Game/Sounds/Foley_Loot/Containers/Mailboxes/NewspaperBox_Search_Cue
root=QualityLevel_0
nodes [
 {id=Mixer_0, children [Modulator_0,Modulator_2], properties {inputVolume [0.8,1]}}
 {id=Mixer_4, children [Modulator_18,Random_14], properties {inputVolume [0.8,1]}}
 {id=Modulator_0, children [Random_0], properties {pitchMax=1.07, pitchMin=0.92, volumeMax=1.07, volumeMin=0.92}}
 {id=Modulator_1, children [WavePlayer_5]}
 {id=Modulator_18, children [Random_12], properties {pitchMax=1.07, pitchMin=0.92, volumeMax=1.07, volumeMin=0.92}}
 {id=Modulator_2, children [WavePlayer_2]}
 {id=Modulator_3, children [WavePlayer_6]}
 {id=QualityLevel_0, children [Mixer_4,Mixer_0,Modulator_1,Modulator_3]}
 {id=Random_0, children [WavePlayer_0,WavePlayer_1], properties {weights [1,1]}}
 {id=Random_12, children [WavePlayer_60, WavePlayer_61], properties {weights [1,1]}}
 {id=Random_14, children [WavePlayer_48, WavePlayer_49, WavePlayer_50], properties {weights [1,1,1]}}
 {id=WavePlayer_0, properties {soundWaveAssetPtr=&/NewspaperBox_Search01}}
 {id=WavePlayer_1, properties {soundWaveAssetPtr=&/NewspaperBox_Search02}}
 {id=WavePlayer_2, properties {soundWaveAssetPtr=/Game/Sounds/Foley_Loot/Containers/Misc/Prop_Gear_Search_05}}
 {id=WavePlayer_48, properties {soundWaveAssetPtr=&Prop_Gear_Search_05}}
 {id=WavePlayer_49, properties {soundWaveAssetPtr=/Game/Sounds/Foley_Loot/Containers/Misc/Prop_Gear_Search_06}}
 {id=WavePlayer_5, properties {soundWaveAssetPtr=&/NewspaperBox_Search_opt}}
 {id=WavePlayer_50, properties {soundWaveAssetPtr=/Game/Sounds/Foley_Loot/Containers/Misc/Prop_Gear_Search_01}}
 {id=WavePlayer_6, properties {soundWaveAssetPtr=&NewspaperBox_Search_opt}}
 {id=WavePlayer_60, properties {soundWaveAssetPtr=&NewspaperBox_Search01}}
 {id=WavePlayer_61, properties {soundWaveAssetPtr=&NewspaperBox_Search02}}
]
properties {
 attenuationSettings=/Game/Sounds/Attenuation/Fort_Prop_Search_Att
 bOverrideConcurrency=true
 maxDistance=1300
 soundClassObject=&Fort_Loot_Collect
 volumeMultiplier=0.7
 concurrencyOverrides {resolutionRule=PreventNew}
}
